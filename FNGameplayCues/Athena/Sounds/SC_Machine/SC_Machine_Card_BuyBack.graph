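# Sound cue for the finished reboot card buy-back at the SC_Machine, using the van buy-back attenuation asset.
# A Random node picks one of three equally weighted RebootCard_BuyBack_Finished waves and feeds it through a Modulator at the root.
# The cue plays in an important in-game sound class at reduced volume.
# Overridden concurrency allows one instance per owner and blocks new plays while it is active.
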SoundCue /FNGameplayCues/Athena/Sounds/SC_Machine/SC_Machine_Card_BuyBack
root=Modulator_0
nodes [
 {id=Modulator_0, children [Random_0]}
 {id=Random_0, children [WavePlayer_1,WavePlayer_2,WavePlayer_3], properties {bShouldExcludeFromBranchCulling=true, weights [1,1,1]}}
 {id=WavePlayer_1, properties {soundWaveAssetPtr=&/RebootCard_BuyBack_Finished_03}}
 {id=WavePlayer_2, properties {soundWaveAssetPtr=&/RebootCard_BuyBack_Finished_02}}
 {id=WavePlayer_3, properties {soundWaveAssetPtr=&/RebootCard_BuyBack_Finished_01}}
]
properties {
 attenuationSettings=/Game/Athena/Sounds/SC_Machine/ATT_Van_Buyback
 bOverrideConcurrency=true
 maxDistance=14000
 soundClassObject=&Fort_Mission_Ingame_Important_PSM
 volumeMultiplier=0.4
 attenuationOverrides {falloffDistance=1500, attenuationShapeExtents {x=1000}}
 concurrencyOverrides {bLimitToOwner=true, maxCount=1, resolutionRule=PreventNew}
}
