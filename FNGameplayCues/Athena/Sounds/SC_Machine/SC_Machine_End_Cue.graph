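# Sound cue that plays the end sound of the SC Machine in Athena.
# A mixer layers the end wave at reduced level with two charge-complete waves.
# The power charge-complete layer is pitched to half speed by a fixed modulator, then shaped by an envelope that holds and fades out under two seconds.
# The cue uses an important in-game mission sound class and overrides its assigned close-range attenuation with its own falloff.
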SoundCue /FNGameplayCues/Athena/Sounds/SC_Machine/SC_Machine_End_Cue
root=Mixer_0
nodes [
 {
  id=Enveloper_0
  children [Modulator_0]
  properties {
   pitchCurve {editorCurveData {keys [{time=0}]}}
   volumeCurve {editorCurveData {keys [<at 0 1>,<at 0.7656 0.9981>,<at 1.1796 0.5386>,<at 1.7618 0>]}}
  }
 }
 {id=Mixer_0, children [WavePlayer_1,Enveloper_0,WavePlayer_2], properties {inputVolume [0.5,1,1]}}
 {id=Modulator_0, children [WavePlayer_0], properties {pitchMax=0.5, pitchMin=0.5}}
 {id=WavePlayer_0, properties {soundWaveAssetPtr=/Game/Athena/Sounds/SC_Machine/SC_Machine_Power_Charge_Complete}}
 {id=WavePlayer_1, properties {soundWaveAssetPtr=/Game/Athena/Sounds/SC_Machine/SC_Machine_End}}
 {id=WavePlayer_2, properties {soundWaveAssetPtr=&/SC_Machine_Redux_Charge_Complete}}
]
properties {
 attenuationSettings=/Game/Athena/Sounds/SupplyDrops/Fort_SC_Machine_Close_Athena_Att
 bOverrideAttenuation=true
 maxDistance=4000
 soundClassObject=&Fort_Mission_Ingame_Important_PSM
 volumeMultiplier=0.28
 attenuationOverrides {falloffDistance=2000, attenuationShapeExtents {x=2000}}
}
