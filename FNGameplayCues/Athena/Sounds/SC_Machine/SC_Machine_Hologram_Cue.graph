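# Sound cue for the hologram sound of the SC_Machine in Athena.
# A single SC_Machine_Hologram wave player feeds a Modulator node at the root, which is left at default settings.
# The cue plays at reduced volume through an important in-game mission sound class.
# It borrows a close-range SC_Machine attenuation asset stored with the supply drop sounds.
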SoundCue /FNGameplayCues/Athena/Sounds/SC_Machine/SC_Machine_Hologram_Cue
root=Modulator_0
nodes [
 {id=Modulator_0, children [WavePlayer_5]}
 {id=WavePlayer_5, properties {soundWaveAssetPtr=/Game/Athena/Sounds/SC_Machine/SC_Machine_Hologram}}
]
properties {
 attenuationSettings=/Game/Athena/Sounds/SupplyDrops/Fort_SC_Machine_Close_Athena_Att
 maxDistance=5000
 soundClassObject=&Fort_Mission_Ingame_Important_PSM
 volumeMultiplier=0.4
}
